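# Physics asset for the dynamic right-side gun bones of the Convoy Tarantula medium male player body cosmetic.
# The dyn_gun_r body uses two angled capsules as its shape, while dyn_gun_r_MSTR carries a tiny capsule.
# A single constraint joins the two bodies, with both swing limits at zero and a modest twist limit, suggesting a hinge-like sway about one axis.
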PhysicsAsset /BRCosmetics/Characters/Player/Male/Medium/Bodies/M_MED_Convoy_Tarantula/Meshes/M_MED_Convoy_Tarantula_Dyn
bodies {
 dyn_gun_r [
  {shapeType=Capsule, length=21.0447, radius=4.2494, center=<3.1213 -2.2093 2.7579>, rotation=<rot 90 0 0>}
  {shapeType=Capsule, length=13.9375, radius=3.875, center=<-7.3006 4.5949 3.6759>, rotation=<rot -90 -45 -20.0063>}
 ]
 dyn_gun_r_MSTR [{shapeType=Capsule, length=0.5, radius=0.5}]
}
constraints [{bone1Name=dyn_gun_r, bone2Name=dyn_gun_r_MSTR, swing1LimitDegrees=0, swing2LimitDegrees=0, twistLimitDegrees=25}]
