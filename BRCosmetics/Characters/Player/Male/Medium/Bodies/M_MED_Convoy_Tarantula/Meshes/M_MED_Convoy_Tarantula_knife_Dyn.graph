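# Physics asset for the dynamic left-side knife mesh on the Convoy Tarantula medium male player body cosmetic.
# It has two capsule bodies: a long, thin one along the knife bone and a tiny one on its master bone.
# A single constraint joins them, limiting the knife to slight swing on one axis and a small twist, with the other swing axis held at zero degrees.
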PhysicsAsset /BRCosmetics/Characters/Player/Male/Medium/Bodies/M_MED_Convoy_Tarantula/Meshes/M_MED_Convoy_Tarantula_knife_Dyn
bodies {
 dyn_knife_l [{shapeType=Capsule, length=35.5447, radius=2.0931, center=<3.2066 0 0>, rotation=<rot 90 0 0>}]
 dyn_knife_l_MSTR [{shapeType=Capsule, length=0.5, radius=0.5}]
}
constraints [{bone1Name=dyn_knife_l_MSTR, bone2Name=dyn_knife_l, swing1LimitDegrees=5, swing2LimitDegrees=0, twistLimitDegrees=10}]
