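# Animation Blueprint for the Convoy Tarantula male medium body cosmetic, adding deformation and secondary motion to the pose copied from the attached parent mesh.
# Two Control Rigs (twists, then sculpters) feed a weighted five-bone constraint on the buckle bone.
# Crouch begin/end events switch an alpha that applies a small additive buckle offset.
# Two RigidBody nodes then simulate dynamics, including a knife, each in a thigh-twist bone's space.
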AnimBlueprint /BRCosmetics/Characters/Player/Male/Medium/Bodies/M_MED_Convoy_Tarantula/Meshes/M_MED_Convoy_Tarantula_AnimBP
target_skeleton=&/M_MED_Convoy_Tarantula_Skeleton
graphs [
 AnimGraph {nodes [
   n1 = Control Rig(Source=Copy Pose From Mesh() {bUseAttachedParent=true}) {
    alphaBoolBlend {blendOption=Linear}
    controlRigAssetReference {blueprintRigClass=&/M_MED_Convoy_Tarantula_Deform_CtrlRig_twists.M_MED_Convoy_Tarantula_Deform_CtrlRig_twists_C}
   }
   n2 = Control Rig!(
    Source=@n1
    Alpha
    KneeRot_fr_l_114deg:Vector
    KneeTrans_fr_l_114deg:Vector="0.258,-3.591,0.009"
    KneeScale_fr_l_114deg:Vector="1,1,1"
    KneeScale_fr_l_130deg:Vector="1,1,1"
    KneeRot_fr_l_130deg:Vector="0,0,-51.2"
    KneeTrans_fr_l_130deg:Vector="0.258,-3.591,0.009"
   ) -> (Pose) {
    alphaBoolBlend {blendOption=Linear}
    controlRigAssetReference {blueprintRigClass=&/M_MED_Convoy_Tarantula_Deform_CtrlRig_sculpters.M_MED_Convoy_Tarantula_Deform_CtrlRig_sculpters_C}
   }
   n3 = "Constraint - dyn_buckle_MSTR (PPPPP)"!(
    ComponentPose=Local To Component(LocalPose=@n2)
    Alpha=1
    ConstraintWeights_0=0.1645
    ConstraintWeights_1:float=0.1104
    ConstraintWeights_2:float=0.015
    ConstraintWeights_3:float=0.1176
    ConstraintWeights_4:float=0.5926
   ) -> (Pose) {
    alphaBoolBlend {blendOption=Linear}
    boneToModify {boneName=dyn_buckle_MSTR}
    constraintSetup [
     {targetBone {boneName=spine_01}}
     {targetBone {boneName=deform_thigh_fr_r}}
     {targetBone {boneName=thigh_twist_01_l}}
     {targetBone {boneName=thigh_twist_01_r}}
     {targetBone {boneName=pelvis}}
    ]
    constraintWeights [1,1,1,1,1]
   }
   n4 = "Transform (Modify) Bone - Bone: dyn_buckle_MSTR"(ComponentPose=@n3, Alpha=Get crouch alpha() -> (crouch alpha:double), Translation="0.000,1,0.000", Rotation="0,0,0", Scale="1,1,1") {
    rotationMode=BMM_Additive
    rotationSpace=BCS_BoneSpace
    translationMode=BMM_Additive
    translationSpace=BCS_BoneSpace
    alphaBoolBlend {blendOption=Linear}
    boneToModify {boneName=dyn_buckle_MSTR}
    translation {y=1}
   }
   n5 = RigidBody!(ComponentPose=@n4, Alpha, OverrideWorldGravity:Vector="0,0,-980", ExternalForce) -> (Pose) {
    bClampLinearTranslationLimitToRefPose=true
    bOverrideWorldGravity=true
    overridePhysicsAsset=&/M_MED_Convoy_Tarantula_Dyn
    simulationSpace=BaseBoneSpace
    alphaBoolBlend {blendOption=Linear}
    baseBoneRef {boneName=thigh_twist_01_r}
    componentAppliedLinearAccClamp=<3000 3000 3000>
    componentLinearAccScale {z=0.5}
    componentLinearVelScale {z=0.5}
    overrideWorldGravity {z=-980}
    simSpaceSettings {
     maxAngularAcceleration=1000
     maxAngularVelocity=3
     maxLinearAcceleration=250
     maxLinearVelocity=500
     velocityScaleZ=0
     worldAlpha=1
     externalLinearDragV=<0.5 0.5>
    }
   }
   n6 = RigidBody(ComponentPose=@n5, OverrideWorldGravity="0,0,-980") {
    bClampLinearTranslationLimitToRefPose=true
    bOverrideWorldGravity=true
    overridePhysicsAsset=&/M_MED_Convoy_Tarantula_knife_Dyn
    simulationSpace=BaseBoneSpace
    alphaBoolBlend {blendOption=Linear}
    baseBoneRef {boneName=thigh_twist_01_l}
    componentAppliedLinearAccClamp=<3000 3000 3000>
    componentLinearAccScale=<0.3 0.3 0.3>
    componentLinearVelScale=<0.3 0.3 0.3>
    overrideWorldGravity {z=-980}
    simSpaceSettings {maxAngularAcceleration=1000, maxAngularVelocity=3, maxLinearAcceleration=250, maxLinearVelocity=500, velocityScaleZ=0, worldAlpha=1}
   }
   Output Pose(Result=Component To Local(ComponentPose=@n6)) {layerGroup=None}
  ]}
 EventGraph {nodes [
   Event OnBeginCrouching()
   > Set crouch alpha(crouch alpha:double=1) -> (Output_Get:double)
   Event OnEndCrouching()
   > Set crouch alpha(crouch alpha=0)
   Event BlueprintUpdateAnimation()
  ]}
]
variables [crouch alpha {type=number}]
