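# Physics asset of collision shapes for the cape on the Convoy Tarantula medium male player body, likely for cloth simulation.
# Capsules cover the pelvis, spine, neck, clavicles and thighs, with a sphere on the head.
# Spine and thigh bones stack multiple offset capsules to shape the torso and upper legs.
# Arm and calf bodies have no shapes, and no constraints exist, so it is not a ragdoll setup.
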PhysicsAsset /BRCosmetics/Characters/Player/Male/Medium/Bodies/M_MED_Convoy_Tarantula/Meshes/M_MED_Convoy_Tarantula_CapeColliders
bodies {
 calf_l []
 calf_r []
 clavicle_l [{shapeType=Capsule, length=9.5449, radius=9.3615, center=<10.4142 -0.2527 -0.0953>, rotation=<rot 75.9374 -0.0001 -180>}]
 clavicle_r [{shapeType=Capsule, length=9.5449, radius=9.3615, center=<-10.4142 0.2527 0.0953>, rotation=<rot 75.9373 -0.0001 0>}]
 head [{shapeType=Sphere, radius=17.3125, center=<11.028 -0.174 0>}]
 lowerarm_l []
 lowerarm_r []
 neck_02 [{shapeType=Capsule, length=13.4375, radius=7.3175, center=<-0.2642 -0.3408 0>, rotation=<rot 90 -171.1193 -175.3978>}]
 pelvis [{shapeType=Capsule, length=8.6562, radius=13.5388, center=<-2.6215 -1.3695 0>}]
 spine_01 [{shapeType=Capsule, length=10.7187, radius=12.4138, center=<-2.6215 -0.7156 0>}]
 spine_02 [
  {shapeType=Capsule, length=10.5937, radius=10.945, center=<-2.6215 0.4993 0.0001>}
  {shapeType=Capsule, length=29.875, radius=11.1013, center=<2.5024 0.4992 3.9825>, rotation=<rot 90 0.0049 0.0049>}
  {shapeType=Capsule, length=29.875, radius=11.1013, center=<2.4704 0.4991 -3.9825>, rotation=<rot 90 -0.0396 -0.0396>}
 ]
 spine_03 [{shapeType=Capsule, length=9.5312, radius=10.195, center=<-1.953 0.1364 0.0001>}]
 spine_04 [{shapeType=Capsule, length=10.5312, radius=12.1638, center=<-0.0327 -0.4036 0.0002>}]
 spine_05 [
  {shapeType=Capsule, length=11.75, radius=13.695, center=<-3.1632 -0.705 0>}
  {shapeType=Capsule, length=16.7812, radius=12.4375, center=<1.6481 -0.516 0>}
 ]
 thigh_l [
  {shapeType=Capsule, length=22.7188, radius=5.875, center=<-19.358 -2.7026 -10.1133>, rotation=<rot 85.0111 0 -168.8026>}
  {shapeType=Capsule, length=7.1875, radius=5.125, center=<-6.9092 -6.9909 -10.5144>, rotation=<rot -1.8998 16.4833 96.3942>}
 ]
 thigh_r [
  {shapeType=Capsule, length=22.7188, radius=5.875, center=<19.358 2.7026 10.1133>, rotation=<rot 85.011 0 11.1973>}
  {shapeType=Capsule, length=7.1875, radius=5.125, center=<6.9092 6.9909 10.5144>, rotation=<rot -1.8998 16.4833 -83.6058>}
 ]
 upperarm_l []
 upperarm_r []
}
constraints []
